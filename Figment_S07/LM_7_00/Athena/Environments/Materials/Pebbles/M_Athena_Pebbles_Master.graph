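# Master material for Athena environment pebbles, a Default Lit surface usable on instanced static meshes.
# The pebble diffuse texture gets two tints, Diffuse_Tint and Arid_Tint, blended by the Swamp output of the landscape biome mask function.
# That result lerps toward a flat near-white using the holiday snowfall mask, multiplied by clamped vertex normal Z so snow favors upward-facing surfaces.
# A normal map and a constant are also present.
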Material /Figment_S07/LM_7_00/Athena/Environments/Materials/Pebbles/M_Athena_Pebbles_Master
graph {
 properties {bUsedWithInstancedStaticMeshes=true, pixelDepthOffsetMode=PDOM_Legacy, usedShadingModels=Default Lit}
 nodes [
  Constant() {r=0.6095}
  n1 = TextureSample() {texture=/Figment_S07/LM_7_00/Athena/Environments/Textures/Pebbles/T_Athena_Pebbles_D}
  n2 = VectorParameter() {parameterName=Diffuse_Tint, defaultValue=<rgba 1 0.8183 0.7059 0>}
  n3 = VectorParameter() {parameterName=Arid_Tint, defaultValue=<rgba 2 0.7697 0.3058 1>}
  n4 = ComponentMask(None=VertexNormalWS()) {b=true}
  LinearInterpolate(
   A=LinearInterpolate(
    A=Multiply(A=@n1.RGB, B=@n2.RGB)
    B=Multiply(A=@n1.RGB, B=@n3.RGB)
    Alpha=MaterialFunctionCall() {materialFunction=/Figment_S07/LM_7_00/Athena/Environments/Landscape/MaterialFunctions/MF_Athena_BiomeMask}.Swamp
   )
   Alpha=Multiply(
    A=Clamp(None=@n4)
    B=MaterialFunctionCall!(
     UseHeightLerp
     UV_WorldMask
     Use_ExtraSnowMask
     LandscapeBlend_Height
     Landscape_RockMask
     Snow_Mask
     Simple_UseWorldZNormal
     Simple_WorldNormal
     WorldPosition
    ) {materialFunction=/Figment_S07/LM_7_00/Athena/Environments/Landscape/MaterialFunctions/Snow/MF_Athena_Holiday_SnowFall}.SnowMask
   )
  ) {constB=0.8}
  TextureSample() {samplerType=SAMPLERTYPE_Normal, texture=/Figment_S07/LM_7_00/Athena/Environments/Textures/Pebbles/T_Athena_Pebbles_N}
 ]
}
